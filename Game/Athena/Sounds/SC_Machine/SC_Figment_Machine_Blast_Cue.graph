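# Sound cue for the Figment Machine blast, layering close and distant versions under a mission-important sound class with long audible range.
# The close layer mixes two blast waves with a beam wave, whose volume envelope holds and then fades out, through a Modulator and close attenuation.
# The distant layer mixes a distant blast with quieter copies of both close blasts through its own Modulator and rift distant attenuation.
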SoundCue /Game/Athena/Sounds/SC_Machine/SC_Figment_Machine_Blast_Cue
root=Mixer_1
nodes [
 {id=Attenuation_0, children [Modulator_1], properties {attenuationSettings=&/Van_Redux_Rift_Distant_Att}}
 {id=Attenuation_1, children [Modulator_0], properties {attenuationSettings=/Game/Athena/Sounds/SupplyDrops/Fort_SC_Machine_Close_Athena_Att}}
 {
  id=Enveloper_0
  children [WavePlayer_4]
  properties {
   pitchCurve {editorCurveData {keys [{time=0}]}}
   volumeCurve {editorCurveData {keys [<at 0 1>,<at 2.4527 0.9981>,<at 2.8012 0.6131>,<at 3.3144 0.2457>,<at 3.5742 0.1093>,<at 3.8087 0.0235>,<at 4.0598 -0.0013>]}}
  }
 }
 {id=Mixer_0, children [WavePlayer_2,WavePlayer_6,WavePlayer_5], properties {inputVolume [0.75,0.1,0.45]}}
 {id=Mixer_1, children [Attenuation_1, Attenuation_0], properties {inputVolume [1,1]}}
 {id=Mixer_2, children [WavePlayer_3,Enveloper_0,WavePlayer_0], properties {inputVolume [0.5,0.5,1]}}
 {id=Modulator_0, children [Mixer_2]}
 {id=Modulator_1, children [Mixer_0]}
 {id=WavePlayer_0, properties {soundWaveAssetPtr=&/SC_Machine_Redux_Blast_01}}
 {id=WavePlayer_2, properties {soundWaveAssetPtr=&/SC_Machine_Redux_Blast_Dstnt}}
 {id=WavePlayer_3, properties {soundWaveAssetPtr=&/SC_Machine_Blast}}
 {id=WavePlayer_4, properties {soundWaveAssetPtr=&/SC_Machine_Blast_Beam}}
 {id=WavePlayer_5, properties {soundWaveAssetPtr=&SC_Machine_Redux_Blast_01}}
 {id=WavePlayer_6, properties {soundWaveAssetPtr=&SC_Machine_Blast}}
]
properties {maxDistance=40000, soundClassObject=&Fort_Mission_Ingame_Important_PSM, volumeMultiplier=0.4}
