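# Sound cue for the completion sound of the Figment variant of the SC Machine in Athena, routed to an important in-game mission sound class.
# A mixer layers two waves, each through its own attenuation node: the Figment completion sound with a close activation attenuation and a distant completion sound with a rift distant attenuation.
# Distance behavior comes from these per-layer nodes, giving separate near and far versions under one long-range cue.
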SoundCue /Game/Athena/Sounds/SC_Machine/SC_Figment_Machine_Complete_Cue
root=Mixer_0
nodes [
 {id=Attenuation_0, children [WavePlayer_3], properties {attenuationSettings=&/Van_Redux_Rift_Distant_Att}}
 {id=Attenuation_1, children [WavePlayer_1], properties {attenuationSettings=&/Fort_Van_Activate_Close_Att}}
 {id=Mixer_0, children [Attenuation_0, Attenuation_1], properties {inputVolume [1,1]}}
 {id=WavePlayer_1, properties {soundWaveAssetPtr=&/SC_Machine_Complete_Figment}}
 {id=WavePlayer_3, properties {soundWaveAssetPtr=&/SC_Machine_Complete_Distant}}
]
properties {
 maxDistance=40000
 soundClassObject=&Fort_Mission_Ingame_Important_PSM
 volumeMultiplier=0.8
 attenuationOverrides {bAttenuate=false, bEnableReverbSend=false, bSpatialize=false, attenuationShapeExtents {x=1500}}
}
