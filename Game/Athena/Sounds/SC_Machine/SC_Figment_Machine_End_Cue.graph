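# Sound cue for the end of a Figment machine sequence, layering a machine end sound with a power charge complete sound.
# A mixer plays both waves together, with the end sound at reduced input volume under the charge-complete wave.
# It routes to an important in-game mission sound class and overrides a supply-drop machine close-range attenuation with its own falloff.
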SoundCue /Game/Athena/Sounds/SC_Machine/SC_Figment_Machine_End_Cue
root=Mixer_0
nodes [
 {id=Mixer_0, children [WavePlayer_1,WavePlayer_0], properties {inputVolume [0.35,1]}}
 {id=WavePlayer_0, properties {soundWaveAssetPtr=&/SC_Machine_Power_Charge_Complete}}
 {id=WavePlayer_1, properties {soundWaveAssetPtr=&/SC_Machine_End}}
]
properties {
 attenuationSettings=/Game/Athena/Sounds/SupplyDrops/Fort_SC_Machine_Close_Athena_Att
 bOverrideAttenuation=true
 maxDistance=4000
 soundClassObject=&Fort_Mission_Ingame_Important_PSM
 volumeMultiplier=0.4
 attenuationOverrides {falloffDistance=2000, attenuationShapeExtents {x=2000}}
}
