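# SC_Machine_Activate_Cue is the activation sound for Athena's SC machine, which uses van attenuation assets, in an important in-game mission sound class.
# A top mixer plays close and distant branches together, each shaped by its own attenuation node.
# The close branch layers a quieter card insert with activate and power-up waves.
# The distant branch reuses those activate and power-up waves without the card insert.
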SoundCue /Game/Athena/Sounds/SC_Machine/SC_Machine_Activate_Cue
root=Mixer_0
nodes [
 {id=Attenuation_0, children [Mixer_2], properties {attenuationSettings=&/Van_Redux_Activate_Distant_Att}}
 {id=Attenuation_1, children [Mixer_1], properties {attenuationSettings=&/Fort_Van_Activate_Close_Att}}
 {id=Mixer_0, children [Attenuation_1, Attenuation_0], properties {inputVolume [1,1]}}
 {id=Mixer_1, children [WavePlayer_1,WavePlayer_2,WavePlayer_4], properties {inputVolume [0.35,1,1]}}
 {id=Mixer_2, children [WavePlayer_0,WavePlayer_5], properties {inputVolume [1,1]}}
 {id=WavePlayer_0, properties {soundWaveAssetPtr=&/SC_Machine_Redux_Activate}}
 {id=WavePlayer_1, properties {soundWaveAssetPtr=&/SC_Machine_Card_Insert_02}}
 {id=WavePlayer_2, properties {soundWaveAssetPtr=&SC_Machine_Redux_Activate}}
 {id=WavePlayer_4, properties {soundWaveAssetPtr=&/SC_Machine_Red_Power_Up_Start}}
 {id=WavePlayer_5, properties {soundWaveAssetPtr=&SC_Machine_Red_Power_Up_Start}}
]
properties {
 maxDistance=14000
 soundClassObject=&Fort_Mission_Ingame_Important_PSM
 volumeMultiplier=0.4
 attenuationOverrides {bAttenuate=false, bEnableReverbSend=false, bSpatialize=false, falloffDistance=2000, attenuationShapeExtents {x=1500}}
}
