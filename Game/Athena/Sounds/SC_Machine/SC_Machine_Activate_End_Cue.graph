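# Sound cue for the end of an SC Machine's activation, layering three one-shot waves through a mixer.
# The activate-end wave plays alongside a half-volume power-up-start wave and a hologram wave.
# The hologram layer runs through an enveloper that holds, then fades it out within a second, and a modulator that adjusts its pitch.
# The cue plays at reduced volume in the Fort_Physics_Impact sound class with its own distance attenuation override.
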SoundCue /Game/Athena/Sounds/SC_Machine/SC_Machine_Activate_End_Cue
root=Mixer_0
nodes [
 {
  id=Enveloper_0
  children [WavePlayer_1]
  properties {pitchCurve {editorCurveData {keys [{time=0}]}}, volumeCurve {editorCurveData {keys [<at 0 1>,<at 0.3283 1>,<at 0.7221 0>]}}}
 }
 {id=Mixer_0, children [WavePlayer_0,Modulator_0,WavePlayer_2], properties {inputVolume [1,1,0.5]}}
 {id=Modulator_0, children [Enveloper_0], properties {pitchMax=0.95}}
 {id=WavePlayer_0, properties {soundWaveAssetPtr=&/SC_Machine_Redux_Activate_End}}
 {id=WavePlayer_1, properties {soundWaveAssetPtr=&/SC_Machine_Hologram}}
 {id=WavePlayer_2, properties {soundWaveAssetPtr=&/SC_Machine_Red_Power_Up_Start}}
]
properties {
 bOverrideAttenuation=true
 maxDistance=3500
 soundClassObject=/Game/Sounds/FortSoundClassesAndModes/Fort_Physics_Impact
 volumeMultiplier=0.35
 attenuationOverrides {falloffDistance=2000, attenuationShapeExtents {x=1500}}
}
