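# Looping ambient sound cue for the SC Machine prop, routed to the base ambient sound class.
# A mixer blends two looping waves, each boosted: an idle layer and a close-up layer nested under its own close-focus prop attenuation.
# The cue overrides attenuation with HRTF spatialization, listener focus that quiets the sound when out of focus, and occlusion with low-pass filtering.
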SoundCue /Game/Athena/Sounds/SC_Machine/SC_Machine_Ambient_Loop_Cue
root=Mixer_0
nodes [
 {id=Attenuation_1, children [WavePlayer_4], properties {attenuationSettings=/Game/Sounds/Attenuation/Fort_Prop_Close_Focus_Att}}
 {id=Mixer_0, children [WavePlayer_1, Attenuation_1], properties {inputVolume [2,1.5]}}
 {id=WavePlayer_1, properties {bLooping=true, soundWaveAssetPtr=&/ImmunityCarriage_Map_Idle}}
 {id=WavePlayer_4, properties {bLooping=true, soundWaveAssetPtr=&/SC_Machine_Ambient_Close_Loop}}
]
properties {
 bOverrideAttenuation=true
 maxDistance=3000
 soundClassObject=/Game/Sounds/FortSoundClassesAndModes/Fort_12_Base_Ambient
 volumeMultiplier=0.624
 attenuationOverrides {
  bEnableListenerFocus=true
  bEnableOcclusion=true
  falloffDistance=2000
  nonFocusVolumeAttenuation=0.2
  occlusionLowPassFilterFrequency=2000
  occlusionTraceChannel=ECC_GameTraceChannel6
  spatializationAlgorithm=SPATIALIZATION_HRTF
  attenuationShapeExtents {x=1000}
 }
}
