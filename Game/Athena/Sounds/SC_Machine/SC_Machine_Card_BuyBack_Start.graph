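# Sound cue played when a reboot card buyback starts, judging by its name and Reboot Card waves.
# A Modulator feeds from a Random node that picks one of three equally weighted start variations, with the Random node excluded from branch culling.
# It uses van buyback attenuation and an important in-game mission sound class.
# Overridden concurrency allows one instance per owner and prevents new plays while it is active.
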SoundCue /Game/Athena/Sounds/SC_Machine/SC_Machine_Card_BuyBack_Start
root=Modulator_0
nodes [
 {id=Modulator_0, children [Random_1]}
 {id=Random_1, children [WavePlayer_0,WavePlayer_4,WavePlayer_6], properties {bShouldExcludeFromBranchCulling=true, weights [1,1,1]}}
 {id=WavePlayer_0, properties {soundWaveAssetPtr=&/RebootCard_BuyBack_Start_01}}
 {id=WavePlayer_4, properties {soundWaveAssetPtr=&/RebootCard_BuyBack_Start_02}}
 {id=WavePlayer_6, properties {soundWaveAssetPtr=&/RebootCard_BuyBack_Start_03}}
]
properties {
 attenuationSettings=&/ATT_Van_Buyback
 bOverrideConcurrency=true
 maxDistance=14000
 soundClassObject=&Fort_Mission_Ingame_Important_PSM
 volumeMultiplier=0.35
 attenuationOverrides {falloffDistance=1500, attenuationShapeExtents {x=1000}}
 concurrencyOverrides {bLimitToOwner=true, maxCount=1, resolutionRule=PreventNew}
}
