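# SoundCue for picking up an SC Machine card.
# A Mixer layers the dedicated card pickup wave at near full level with a very quiet backpack collect wave that starts after a fixed one-second delay.
# The cue overrides attenuation for its own falloff and limits concurrency to one instance per owner, rejecting new plays while one is active.
# It routes to an important in-game mission sound class.
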SoundCue /Game/Athena/Sounds/SC_Machine/SC_Machine_Card_PickUp_Cue
root=Mixer_0
nodes [
 {id=Delay_0, children [WavePlayer_0], properties {delayMax=1, delayMin=1}}
 {id=Mixer_0, children [WavePlayer_5,Delay_0], properties {inputVolume [1.1,0.1]}}
 {id=WavePlayer_0, properties {soundWaveAssetPtr=/Game/Sounds/Fort_GamePlay_Sounds/Loot/Collect_ToBackpack07}}
 {id=WavePlayer_5, properties {soundWaveAssetPtr=&/SC_Machine_Card_PickUp}}
]
properties {
 attenuationSettings=/Game/Sounds/Attenuation/Fort_Prop_Close_Att
 bOverrideAttenuation=true
 bOverrideConcurrency=true
 maxDistance=2500
 soundClassObject=&Fort_Mission_Ingame_Important_PSM
 volumeMultiplier=0.85
 attenuationOverrides {falloffDistance=1500, attenuationShapeExtents {x=1000}}
 concurrencyOverrides {bLimitToOwner=true, maxCount=1, resolutionRule=PreventNew}
}
