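# Sound cue for the completion of the SC Machine in Athena, played on an important in-game mission sound class.
# Mixes a distant branch and a close branch, each with its own attenuation node reusing Van attenuation assets.
# Distant branch layers a dedicated distant wave with one of three equally weighted Redux completion variants.
# Close branch layers a quieter base completion wave with a separate random pick from those same variants.
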SoundCue /Game/Athena/Sounds/SC_Machine/SC_Machine_Complete_Cue
root=Mixer_0
nodes [
 {id=Attenuation_0, children [Mixer_3], properties {attenuationSettings=&/Van_Redux_Rift_Distant_Att}}
 {id=Attenuation_1, children [Mixer_1], properties {attenuationSettings=&/Fort_Van_Activate_Close_Att}}
 {id=Mixer_0, children [Attenuation_0, Attenuation_1], properties {inputVolume [0.55,0.75]}}
 {id=Mixer_1, children [WavePlayer_0,Random_0], properties {inputVolume [0.25,1]}}
 {id=Mixer_3, children [WavePlayer_3,Random_1], properties {inputVolume [1,1]}}
 {id=Random_0, children [WavePlayer_1,WavePlayer_4,WavePlayer_6], properties {weights [1,1,1]}}
 {id=Random_1, children [WavePlayer_7,WavePlayer_8,WavePlayer_9], properties {weights [1,1,1]}}
 {id=WavePlayer_0, properties {soundWaveAssetPtr=&/SC_Machine_Complete}}
 {id=WavePlayer_1, properties {soundWaveAssetPtr=&/SC_Machine_Redux_Complete_01}}
 {id=WavePlayer_3, properties {soundWaveAssetPtr=&/SC_Machine_Redux_Complet_Dstnt}}
 {id=WavePlayer_4, properties {soundWaveAssetPtr=&/SC_Machine_Redux_Complete_02}}
 {id=WavePlayer_6, properties {soundWaveAssetPtr=&/SC_Machine_Redux_Complete_03}}
 {id=WavePlayer_7, properties {soundWaveAssetPtr=&SC_Machine_Redux_Complete_01}}
 {id=WavePlayer_8, properties {soundWaveAssetPtr=&SC_Machine_Redux_Complete_02}}
 {id=WavePlayer_9, properties {soundWaveAssetPtr=&SC_Machine_Redux_Complete_03}}
]
properties {
 maxDistance=40000
 soundClassObject=&Fort_Mission_Ingame_Important_PSM
 volumeMultiplier=0.5
 attenuationOverrides {bAttenuate=false, bEnableReverbSend=false, bSpatialize=false, attenuationShapeExtents {x=1500}}
}
