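# Sound cue for the SC Machine horn alert, layering a close horn and a distant horn as one in-game alert.
# A mixer blends the two wave players at reduced volumes, each through its own attenuation asset, so the close and distant layers carry different distance falloff.
# The cue uses an important mission sound class, with a long maximum audible range and a single-instance concurrency override.
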SoundCue /Game/Athena/Sounds/SC_Machine/SC_Machine_Horn_Alert_Cue
root=Mixer_0
nodes [
 {id=Attenuation_0, children [WavePlayer_0], properties {attenuationSettings=&/Van_Redux_Rift_Distant_Att}}
 {id=Attenuation_1, children [WavePlayer_2], properties {attenuationSettings=/Game/Athena/Sounds/SupplyDrops/Fort_SC_Machine_Close_Athena_Att}}
 {id=Mixer_0, children [Attenuation_1, Attenuation_0], properties {inputVolume [0.1,0.35]}}
 {id=WavePlayer_0, properties {soundWaveAssetPtr=&/SC_Machine_Redux_Horn_Distnt}}
 {id=WavePlayer_2, properties {soundWaveAssetPtr=&/SC_Machine_Redux_Horn}}
]
properties {
 maxDistance=40000
 soundClassObject=&Fort_Mission_Ingame_Important_PSM
 volumeMultiplier=1.3
 attenuationOverrides {falloffDistance=2500, attenuationShapeExtents {x=1500}}
 concurrencyOverrides {maxCount=1}
}
