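# Sound cue for the power-down sound of the Athena SC_Machine.
# A single wave player feeds the SC_Machine_Power_Down wave through a modulator left at default settings.
# Playback uses a close-range SC_Machine attenuation asset from the SupplyDrops folder, routes to the base ambient sound class, and has a boosted volume multiplier.
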SoundCue /Game/Athena/Sounds/SC_Machine/SC_Machine_Power_Down_Cue
root=Modulator_1
nodes [
 {id=Modulator_1, children [WavePlayer_1]}
 {id=WavePlayer_1, properties {soundWaveAssetPtr=&/SC_Machine_Power_Down}}
]
properties {
 attenuationSettings=/Game/Athena/Sounds/SupplyDrops/Fort_SC_Machine_Close_Athena_Att
 maxDistance=5000
 soundClassObject=/Game/Sounds/FortSoundClassesAndModes/Fort_12_Base_Ambient
 volumeMultiplier=2.97
}
