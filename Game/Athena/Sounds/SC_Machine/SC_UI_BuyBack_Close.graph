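# SC_UI_BuyBack_Close is a UI sound cue, named for closing the buyback interface.
# A single wave player plays UI_Card_Buyback_Close through the Fort_UI sound class at reduced volume and lowered pitch.
# Its concurrency override allows one instance per owner and blocks new plays while that instance is active.
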SoundCue /Game/Athena/Sounds/SC_Machine/SC_UI_BuyBack_Close
root=WavePlayer_0
nodes [{id=WavePlayer_0, properties {soundWaveAssetPtr=&/UI_Card_Buyback_Close}}]
properties {
 bOverrideConcurrency=true
 maxDistance=2097152
 pitchMultiplier=0.8
 soundClassObject=&Fort_UI
 volumeMultiplier=0.5
 attenuationOverrides {falloffDistance=1500, attenuationShapeExtents {x=1000}}
 concurrencyOverrides {bLimitToOwner=true, maxCount=1, resolutionRule=PreventNew}
}
